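# Additive, lit particle sprite material, kept in a Test folder, for a spinning star effect.
# A SubUV texture parameter samples the spinning star flipbook, and its RGB is multiplied by the particle color.
# The tinted result feeds the Emissive input of the shared MF_FX_EmissiveAdjustments function.
# Translucency lighting uses the volumetric per-vertex non-directional mode.
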Material /Game/Effects/Fort_Effects/Materials/Test/M_Spinning_Star
graph {
 properties {
  bUsedWithParticleSprites=true
  blendMode=BLEND_Additive
  pixelDepthOffsetMode=PDOM_Legacy
  translucencyLightingMode=TLM_VolumetricPerVertexNonDirectional
  usedShadingModels=Default Lit
 }
 nodes [
  MaterialExpressionTextureSampleParameterSubUV_0 = TextureSampleParameterSubUV!(UVs, Apply View MipBias) {
   parameterName=MaterialExpressionTextureSampleParameterSubUV_0
   samplerType=SAMPLERTYPE_LinearColor
   texture=/Game/Effects/Fort_Effects/Textures/Stars/S_Spinning_Star_SubUV
  }
  MaterialFunctionCall(Emissive=Multiply(A=ParticleColor().RGB, B=@MaterialExpressionTextureSampleParameterSubUV_0.RGB)) {materialFunction=&MF_FX_EmissiveAdjustments}
 ]
}
